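# Control Rig for the female Ceremonial Guard Fencer face accessory (no-mask variant) that corrects the dyn_BACE_hair_p hair base bone as the body poses.
# A fixed offset, rotation and squash applies at full weight, then a second similar corrective layers on top.
# A spherical pose reader on the head, relative to spine_02, drives that second weight, multiplied by a remapped spine_05-to-pelvis Z rotation.
# A third, stronger transform node is left unconnected.
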ControlRigBlueprint /Game/Characters/Player/Female/Medium/Bodies/F_MED_CeremonialGuard_Fencer/Meshes/Parts/F_MED_CeremonialGuard_Fencer_FaceAcc_CtrlRig_Not_mask
graphs [
 RigVMFunctionLibrary {links [], nodes []}
 RigVMModel {nodes [
   RigUnit_BeginExecution = &RigUnit_BeginExecution()
   n1 = &RigUnit_SphericalPoseReader(
    ActiveRegionScaleFactors=(PositiveWidth=1,NegativeWidth=1.3,PositiveHeight=4,NegativeHeight=4)
    ActiveRegionSize=0.14
    Debug=(bDrawDebug=false,bDraw2D=False,bDrawLocalAxes=False,DebugScale=25,DebugSegments=20,DebugThickness=0.25)
    DriverAxis=(X=1,Y=0,Z=0)
    DriverItem=(Type=Bone,Name="head")
    FalloffRegionScaleFactors=(PositiveWidth=1,NegativeWidth=1,PositiveHeight=1,NegativeHeight=1)
    FalloffSize=0.4
    FlipHeightScaling=False
    FlipWidthScaling=False
    OptionalParentItem=(Type=Bone,Name="spine_02")
    RotationOffset=(X=-60,Y=0,Z=13)
    ExecutePin=@RigUnit_BeginExecution.ExecutePin
   )
   > &RigUnit_ModifyTransforms(
    ItemToModify=((Item=(Type=Bone,Name="dyn_BACE_hair_p"),Transform=(Rotation=(X=0,Y=0,Z=0,W=1),Translation=(X=0,Y=0,Z=0),Scale3D=(X=1,Y=1,Z=1))))
    Weight=1
    WeightMinimum=0
    ItemToModify.0.Transform=/Script/RigVM.RigVMFunction_MathTransformFromSRT(Location=(X=-0.2,Y=1.57,Z=-0.5), Rotation=(X=-15,Y=0,Z=15), RotationOrder=XYZ, Scale=(X=0.95,Y=0.85,Z=0.95)) {is=struct RigVMUnitNode}.Transform
   )
   n2 = /Script/RigVM.RigVMFunction_AlphaInterpVector(
    Bias=0
    ClampMax=1
    ClampMin=0
    InRange=(Min=0,Max=1)
    InterpSpeedDecreasing=10
    InterpSpeedIncreasing=10
    OutRange=(Min=0,Max=1)
    Scale=1
    bClampResult=False
    bInterpResult=False
    bMapRange=False
    Value=&RigVMFunction_MathQuaternionToEuler(
     RotationOrder=ZYX
     Value=&RigUnit_GetRelativeTransformForItem(Child=(Type=Bone,Name="spine_05"), Parent=(Type=Bone,Name="pelvis"), bChildInitial=False, bParentInitial=False).RelativeTransform.Rotation
    ).Result
   ) {
    is=struct RigVMUnitNode
    template="AlphaInterp::Execute(in Value,in Scale,in Bias,in bMapRange,in InRange,in OutRange,in bClampResult,in ClampMin,in ClampMax,in bInterpResult,in InterpSpeedIncreasing,in InterpSpeedDecreasing,out Result)"
   }
   Remap = &RigVMFunction_MathDoubleRemap(
    SourceMaximum=1
    SourceMinimum=0
    TargetMaximum=0
    TargetMinimum=1
    bClamp=False
    Value=&RigVMFunction_MathDoubleRemap(SourceMaximum=5, SourceMinimum=10, TargetMaximum=1, TargetMinimum=0, Value=@n2.Result.Z).Result
   )
   Multiply = &RigVMFunction_MathDoubleMul(B=@Remap.Result, A=@n1.OutputParam)
   > &RigUnit_ModifyTransforms(
    ItemToModify=((Item=(Type=Bone,Name="dyn_BACE_hair_p"),Transform=(Rotation=(X=0,Y=0,Z=0,W=1),Translation=(X=0,Y=0,Z=0),Scale3D=(X=1,Y=1,Z=1))))
    WeightMinimum=0
    Weight=@Multiply.Result
    ItemToModify.0.Transform=&RigVMFunction_MathTransformFromSRT(Location=(X=-0.2,Y=2,Z=-0.5), Rotation=(X=-15,Y=0,Z=25), RotationOrder=XYZ, Scale=(X=0.95,Y=0.85,Z=0.95)).Transform
   )
   &RigVMFunction_MathTransformFromSRT(Location=(X=-0.4,Y=2.1,Z=-1), Rotation=(X=-30,Y=0,Z=30), RotationOrder=XYZ, Scale=(X=0.9,Y=0.7,Z=0.9))
  ]}
]
hierarchy [
 root {loc=<0 0 0>, rot=<0 0 0 1>}
 pelvis {parent=root, loc=<0 -0.5301 98.0571>, rot=<0.0127 0.707 -0.0127 -0.707>}
 spine_01 {parent=pelvis, loc=<6.1208 -0.9328 0>, rot=<0 0 0.034 -0.9994>}
 spine_02 {parent=spine_01, loc=<6.5996 -0.2095 0>, rot=<0 0 0.0238 -0.9997>}
 spine_03 {parent=spine_02, loc=<6.4914 -0.1237 0>, rot=<0 0 0.0149 -0.9999>}
 spine_04 {parent=spine_03, loc=<6.9598 0.0794 0>, rot=<0 0 -0.0132 -0.9999>}
 spine_05 {parent=spine_04, loc=<6.9987 0 0>, rot=<0 0 -0.0015 -1>}
 neck_01 {parent=spine_05, loc=<11.3138 0.1529 0>, rot=<0 0 -0.2003 -0.9797>}
 neck_02 {parent=neck_01, loc=<4.5627 0 0>, rot=<0 0 0.0504 -0.9987>}
 head {parent=neck_02, loc=<5.3914 0 0>, rot=<0 0 0.115 -0.9934>}
 dyn_R_hair {parent=head, loc=<2.1125 1.3542 6.3731>, rot=<0.3115 -0.8935 -0.172 0.2741>}
 pasted__dyn_R_hair_end {parent=dyn_R_hair, loc=<2.3084 -0.0001 -0.0003>, rot=<0 0 0 1>}
 dyn_L_hair {parent=head, loc=<2.2326 0.2386 -6.7046>, rot=<0.088 -0.968 0.0574 -0.2281>}
 pasted__dyn_L_hair_end {parent=dyn_L_hair, loc=<2.309 0 0>, rot=<0 0 0 1>}
 dyn_BACE_hair_p {parent=head, loc=<0.6349 -6.0188 0.35>, rot=<0.0085 -0.0152 0.8725 -0.4883>}
 dyn_hair_p {parent=dyn_BACE_hair_p, loc=<-0.0002 -0.0018 0.0003>, rot=<0 0 0 1>}
 pasted__joint3 {parent=dyn_hair_p, loc=<2.7651 2.4055 -2.4596>, rot=<0.7957 0.4989 0.2071 -0.274>}
 pasted__joint2 {parent=pasted__joint3, loc=<1.011 0.7709 2.2478>, rot=<0 0 0 1>}
 dyn_hv1 {parent=dyn_hair_p, loc=<2.7841 2.4482 -2.4734>, rot=<0.7957 0.4989 0.2071 -0.274>}
 dyn_hv_end {parent=dyn_hv1, loc=<1.032 0.7704 2.2742>, rot=<0.0016 0.0014 -0.0009 -1>}
]
